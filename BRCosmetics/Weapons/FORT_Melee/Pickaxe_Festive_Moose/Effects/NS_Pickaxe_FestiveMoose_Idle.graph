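# Idle effect for the Festive Moose pickaxe that makes green, red and yellow lights twinkle on its skeletal mesh.
# Each color has a sprite emitter that repeatedly bursts random counts of glows, cycling through its own socket index range via counters and fading them in and out.
# A matching light emitter per color spawns only on loop zero, stays pinned to an array-selected socket, and animates light exponent with a curve.
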NiagaraSystem /BRCosmetics/Weapons/FORT_Melee/Pickaxe_Festive_Moose/Effects/NS_Pickaxe_FestiveMoose_Idle
dependencies {
 usedDataInterfaces [/Script/Niagara.NiagaraDataInterfaceSimpleCounter, &NiagaraDataInterfaceCurve, &NiagaraDataInterfaceArrayFloat]
 usedDynamicInputs []
 usedModules [&SystemState]
}
emitters {
 GreenLights {
  emitterUpdateScript [
   &EmitterState(
    Life Cycle Mode=::NewEnumerator1
    Loop Duration=&RandomRangeFloat(Minimum=1.3, Maximum=1)
    UseLoopDelay=false
   )
   &SpawnBurst_Instantaneous(
    Spawn Count=1
    Spawn Probability=&MakeCustomFloatFromBool(
     Bool=&SetBoolByIntComparison(A=$Emitter.LoopCount, B=0, Comparison Type=::NewEnumerator4)
     True Float=1
     False Float=0
    )
    Use Spawn Probability=true
   )
  ]
  particleSpawnScript [
   /SetVariables(Particles.SpawnGroup=&MakeIntFromFloat(Float=&SelectFloatFromArray(
      Direct Array Index=&ModuloInt(
       A=/Niagara/DynamicInputs/DataInterface/GenerateIntFromCounter(SimpleCounter:NiagaraDataInterfaceSimpleCounter=DataInterface() {propertyValues {GpuSyncMode=None, InitialValue=0}})
       B=2
      )
      Float Selection Array=DataInterface() {propertyValues {FloatData [8]}}
     ))) {isSetParameters=true, module=SetVariables_#1}
   &InitializeParticle(
    Color=<rgba 0 1.2 0 1>
    Lifetime=&RandomRangeFloat(Minimum=1, Maximum=3)
    Sprite Size Mode=::NewEnumerator3
    Uniform Sprite Size=15
    UsePositionOffset=false
    Write Material Random=true
   )
  ]
  particleUpdateScript [
   &SkeletalMeshLocation(
    Bone / Socket Index=&SpawnGroup()
    Inherit Velocity=false
    Interpolate Mesh Sampling=false
    MaskSpawnGroup=false
    Mesh Sampling Type=::NewEnumerator5
    Skeletal Mesh=$User.SkeletalMesh
    Socket Sampling Mode=::NewEnumerator5
   )
   &ParticleState(Kill Particles When Lifetime Has Elapsed=false, Loop Particles Lifetime=true)
   /SetVariables(Particles.LightExponent:NiagaraFloat=&FloatFromCurve(FloatCurve=DataInterface() {propertyValues {Curve {keys [
         <at 0 1.5 RCIM_Cubic>
         {arriveTangent=-1.893, interpMode=RCIM_Cubic, leaveTangent=-1.893, time=0.1796, value=0.5}
         <at 0.317 0.9 RCIM_Cubic>
         {arriveTangent=0.8784, interpMode=RCIM_Cubic, leaveTangent=0.8784, time=0.5984, value=0.5}
         <at 1 1.5 RCIM_Cubic>
        ]}}})) {isSetParameters=true, module=SetVariables_#2}
  ]
  rendererClasses [&NiagaraLightRendererProperties]
  renderers [{rendererClass=&NiagaraLightRendererProperties, rendererIndex=0}]
 }
 GreenSprites {
  emitterUpdateScript [
   &EmitterState(
    Life Cycle Mode=::NewEnumerator1
    Loop Duration=&RandomRangeFloat(Minimum=1.3, Maximum=1)
    UseLoopDelay=false
   )
   &SpawnBurst_Instantaneous(Spawn Count=&RandomRangeInt(Minimum=1, Maximum=5), Use Spawn Probability=false)
  ]
  particleSpawnScript [
   /SetVariables(Particles.SpawnGroup=&ModuloInt(
     A=&GenerateIntFromCounter(SimpleCounter:NiagaraDataInterfaceSimpleCounter=DataInterface() {propertyValues {GpuSyncMode=None, InitialValue=0}})
     B=5
    )) {isSetParameters=true, module=SetVariables_#1}
   &InitializeParticle(
    Color=<rgba 0 1 0 0.1>
    Lifetime=&RandomRangeFloat(Minimum=1, Maximum=3)
    Sprite Size Mode=::NewEnumerator3
    Uniform Sprite Size=15
    UsePositionOffset=false
    Write Material Random=true
   )
   &CameraOffset(Camera Offset Amount=2)
  ]
  particleUpdateScript [
   &SkeletalMeshLocation(
    Bone / Socket Index=&SpawnGroup()
    Inherit Velocity=false
    Interpolate Mesh Sampling=false
    MaskSpawnGroup=false
    Mesh Sampling Type=::NewEnumerator5
    Skeletal Mesh=$User.SkeletalMesh
    Socket Sampling Mode=::NewEnumerator5
   )
   &ParticleState()
   &ScaleColor(
    Scale Alpha=&FloatFromCurve(
     FloatCurve=DataInterface() {propertyValues {Curve {keys [
         {arriveTangent=4.6356, interpMode=RCIM_Cubic, leaveTangent=4.6356, tangentMode=RCTM_User, time=0, value=0}
         {interpMode=RCIM_Cubic, leaveTangent=-3.6316, tangentMode=RCTM_Break, time=0.5, value=1}
         <at 1 0>
        ]}}}
     Scale Curve=0.15
    )
    ScaleA=true
    ScaleRGB=true
   )
  ]
  rendererClasses [&NiagaraSpriteRendererProperties]
  renderers [{rendererClass=&NiagaraSpriteRendererProperties}]
 }
 RedLights {
  emitterUpdateScript [
   &EmitterState(
    Life Cycle Mode=::NewEnumerator1
    Loop Duration=&RandomRangeFloat(Minimum=1.3, Maximum=1)
    UseLoopDelay=false
   )
   &SpawnBurst_Instantaneous(
    Spawn Count=2
    Spawn Probability=&MakeCustomFloatFromBool(
     Bool=&SetBoolByIntComparison(A=$Emitter.LoopCount, B=0, Comparison Type=::NewEnumerator4)
     True Float=1
     False Float=0
    )
    Use Spawn Probability=true
   )
  ]
  particleSpawnScript [
   /SetVariables(Particles.SpawnGroup=&MakeIntFromFloat(Float=&SelectFloatFromArray(
      Direct Array Index=&ModuloInt(
       A=&GenerateIntFromCounter(SimpleCounter:NiagaraDataInterfaceSimpleCounter=DataInterface() {propertyValues {GpuSyncMode=None, InitialValue=0}})
       B=2
      )
      Float Selection Array=DataInterface() {propertyValues {FloatData [19]}}
     ))) {isSetParameters=true, module=SetVariables_#1}
   &InitializeParticle(Color=<rgba 1.2 0 0 1>, Sprite Size Mode=::NewEnumerator3, Uniform Sprite Size=15, UsePositionOffset=false, Write Material Random=true)
  ]
  particleUpdateScript [
   &SkeletalMeshLocation(
    Bone / Socket Index=&SpawnGroup()
    Inherit Velocity=false
    Interpolate Mesh Sampling=false
    MaskSpawnGroup=false
    Mesh Sampling Type=::NewEnumerator5
    Skeletal Mesh=$User.SkeletalMesh
    Socket Sampling Mode=::NewEnumerator5
   )
   &ParticleState(Kill Particles When Lifetime Has Elapsed=false)
   /SetVariables(Particles.LightExponent:NiagaraFloat=&FloatFromCurve(FloatCurve=DataInterface() {propertyValues {Curve {keys [
         <at 0 1.25>
         {arriveTangent=-1.1043, interpMode=RCIM_Cubic, leaveTangent=-1.1043, time=0.1796, value=0.5}
         <at 0.317 0.9 RCIM_Cubic>
         {arriveTangent=0.5124, interpMode=RCIM_Cubic, leaveTangent=0.5124, time=0.5984, value=0.5}
         <at 1 1.25>
        ]}}})) {isSetParameters=true, module=SetVariables_#2}
  ]
  rendererClasses [&NiagaraLightRendererProperties]
  renderers [{rendererClass=&NiagaraLightRendererProperties, rendererIndex=0}]
 }
 RedSprites {
  emitterUpdateScript [
   &EmitterState(
    Life Cycle Mode=::NewEnumerator1
    Loop Duration=&RandomRangeFloat(Minimum=1.3, Maximum=1)
    UseLoopDelay=false
   )
   &SpawnBurst_Instantaneous(Spawn Count=&RandomRangeInt(Minimum=1, Maximum=7), Use Spawn Probability=false)
  ]
  particleSpawnScript [
   /SetVariables(Particles.SpawnGroup=&Add_Integer(
     A=13
     B=&ModuloInt(
      A=&GenerateIntFromCounter(SimpleCounter:NiagaraDataInterfaceSimpleCounter=DataInterface() {propertyValues {GpuSyncMode=None, InitialValue=0}})
      B=7
     )
    )) {isSetParameters=true, module=SetVariables_#1}
   &InitializeParticle(
    Color=<rgba 1 0 0 0.1>
    Lifetime=&RandomRangeFloat(Minimum=1, Maximum=3)
    Sprite Size Mode=::NewEnumerator3
    Uniform Sprite Size=15
    UsePositionOffset=false
    Write Material Random=true
   )
   &CameraOffset(Camera Offset Amount=2)
  ]
  particleUpdateScript [
   &SkeletalMeshLocation(
    Bone / Socket Index=&SpawnGroup()
    Inherit Velocity=false
    Interpolate Mesh Sampling=false
    MaskSpawnGroup=false
    Mesh Sampling Type=::NewEnumerator5
    Skeletal Mesh=$User.SkeletalMesh
    Socket Sampling Mode=::NewEnumerator5
   )
   &ParticleState()
   &ScaleColor(
    Scale Alpha=&FloatFromCurve(
     FloatCurve=DataInterface() {propertyValues {Curve {keys [
         {arriveTangent=6.7114, interpMode=RCIM_Cubic, leaveTangent=6.7114, tangentMode=RCTM_User, time=0, value=0}
         <at 0.5 1 RCIM_Cubic>
         {arriveTangent=-4.4472, interpMode=RCIM_Cubic, leaveTangent=-4.4472, tangentMode=RCTM_User, time=1, value=0}
        ]}}}
     Scale Curve=0.15
    )
    ScaleA=true
    ScaleRGB=true
   )
  ]
  rendererClasses [&NiagaraSpriteRendererProperties]
  renderers [{rendererClass=&NiagaraSpriteRendererProperties}]
 }
 YellowLights {
  emitterUpdateScript [
   &EmitterState()
   &SpawnBurst_Instantaneous(
    Spawn Count=2
    Spawn Probability=&MakeCustomFloatFromBool(
     Bool=&SetBoolByIntComparison(A=$Emitter.LoopCount, B=0, Comparison Type=::NewEnumerator4)
     True Float=1
     False Float=0
    )
    Use Spawn Probability=true
   )
  ]
  particleSpawnScript [
   /SetVariables(Particles.SpawnGroup=&MakeIntFromFloat(Float=&SelectFloatFromArray(
      Direct Array Index=&ModuloInt(
       A=&GenerateIntFromCounter(SimpleCounter:NiagaraDataInterfaceSimpleCounter=DataInterface() {propertyValues {GpuSyncMode=None, InitialValue=0}})
       B=2
      )
      Float Selection Array=DataInterface() {propertyValues {FloatData [11]}}
     ))) {isSetParameters=true, module=SetVariables_#1}
   &InitializeParticle(
    Color=<rgba 0.8 0.8 0 1>
    Lifetime=&RandomRangeFloat(Minimum=1, Maximum=3)
    Sprite Size Mode=::NewEnumerator3
    Uniform Sprite Size=15
    UsePositionOffset=false
    Write Material Random=true
   )
  ]
  particleUpdateScript [
   &SkeletalMeshLocation(
    Bone / Socket Index=&SpawnGroup()
    Inherit Velocity=false
    Interpolate Mesh Sampling=false
    MaskSpawnGroup=false
    Mesh Sampling Type=::NewEnumerator5
    Skeletal Mesh=$User.SkeletalMesh
    Socket Sampling Mode=::NewEnumerator5
   )
   &ParticleState(Kill Particles When Lifetime Has Elapsed=false)
   /SetVariables(Particles.LightExponent:NiagaraFloat=&FloatFromCurve(FloatCurve=DataInterface() {propertyValues {Curve {keys [
         <at 0.0023 1.6598>
         {arriveTangent=-1.1043, interpMode=RCIM_Cubic, leaveTangent=-1.1043, time=0.1819, value=0.9098}
         <at 0.3192 1.3098 RCIM_Cubic>
         {arriveTangent=0.5124, interpMode=RCIM_Cubic, leaveTangent=0.5124, time=0.6006, value=0.9098}
         <at 1.0023 1.6598>
        ]}}})) {isSetParameters=true, module=SetVariables_#2}
  ]
  rendererClasses [&NiagaraLightRendererProperties]
  renderers [{rendererClass=&NiagaraLightRendererProperties, rendererIndex=0}]
 }
 YellowSprites {
  emitterUpdateScript [
   &EmitterState(
    Life Cycle Mode=::NewEnumerator1
    Loop Duration=&RandomRangeFloat(Minimum=1.3, Maximum=1)
    UseLoopDelay=false
   )
   &SpawnBurst_Instantaneous(Spawn Count=&RandomRangeInt(Minimum=1, Maximum=8), Use Spawn Probability=false)
  ]
  particleSpawnScript [
   /SetVariables(Particles.SpawnGroup=&Add_Integer(
     A=5
     B=&ModuloInt(
      A=&GenerateIntFromCounter(SimpleCounter:NiagaraDataInterfaceSimpleCounter=DataInterface() {propertyValues {GpuSyncMode=None, InitialValue=0}})
      B=8
     )
    )) {isSetParameters=true, module=SetVariables_#1}
   &InitializeParticle(
    Color=<rgba 1 1 0 0.1>
    Lifetime=&RandomRangeFloat(Minimum=1, Maximum=3)
    Sprite Size Mode=::NewEnumerator3
    Uniform Sprite Size=15
    UsePositionOffset=false
    Write Material Random=true
   )
   &CameraOffset(Camera Offset Amount=2)
  ]
  particleUpdateScript [
   &SkeletalMeshLocation(
    Bone / Socket Index=&SpawnGroup()
    Inherit Velocity=false
    Interpolate Mesh Sampling=false
    MaskSpawnGroup=false
    Mesh Sampling Type=::NewEnumerator5
    Skeletal Mesh=$User.SkeletalMesh
    Socket Sampling Mode=::NewEnumerator5
   )
   &ParticleState()
   &ScaleColor(
    Scale Alpha=&FloatFromCurve(
     FloatCurve=DataInterface() {propertyValues {Curve {keys [
         {arriveTangent=6.7114, interpMode=RCIM_Cubic, leaveTangent=6.7114, tangentMode=RCTM_User, time=0, value=0}
         <at 0.5 1 RCIM_Cubic>
         {arriveTangent=-4.4472, interpMode=RCIM_Cubic, leaveTangent=-4.4472, tangentMode=RCTM_User, time=1, value=0}
        ]}}}
     Scale Curve=0.15
    )
    ScaleA=true
    ScaleRGB=true
   )
  ]
  rendererClasses [&NiagaraSpriteRendererProperties]
  renderers [{rendererClass=&NiagaraSpriteRendererProperties}]
 }
}
summary {
 systemName=NS_Pickaxe_FestiveMoose_Idle
 emitters [
  {emitterName=GreenSprites, rendererClasses [&NiagaraSpriteRendererProperties]}
  {emitterName=YellowLights, rendererClasses [&NiagaraLightRendererProperties]}
  {emitterName=RedSprites, rendererClasses [&NiagaraSpriteRendererProperties]}
  {emitterName=GreenLights, rendererClasses [&NiagaraLightRendererProperties]}
  {emitterName=YellowSprites, rendererClasses [&NiagaraSpriteRendererProperties]}
  {emitterName=RedLights, rendererClasses [&NiagaraLightRendererProperties]}
 ]
 userVariables [
  User.bIsFrontEndPreview {type=NiagaraBool}
  User.SkeletalMesh {
   type=NiagaraDataInterfaceSkeletalMesh
   defaultValue {
    struct=NiagaraExt_VariableValue_DataInterface
    value {dataInterface=&self:NiagaraDataInterfaceSkeletalMesh_0}
   }
  }
 ]
}
