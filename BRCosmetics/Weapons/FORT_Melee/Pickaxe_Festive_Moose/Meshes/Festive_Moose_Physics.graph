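# Physics asset for the Festive Moose pickaxe, setting up secondary motion for the string lights on its head.
# Six left and right light chains of box or capsule bodies hang from a small dyn_head body, which joins the handle capsule with free swing and twist.
# Angular limits vary per chain: the short box chains are stiff, while capsule chains swing wider, with the right-side ones tightening toward their tips.
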PhysicsAsset /BRCosmetics/Weapons/FORT_Melee/Pickaxe_Festive_Moose/Meshes/Festive_Moose_Physics
bodies {
 Melee_TwoHanded_Handle [{shapeType=Capsule, length=116.3073, radius=3.1669, center=<0.6109 -0.3724 24.6668>, rotation=<rot 0.4381 -90.0238 -0.5131>}]
 dyn_StringLights_l_a00 [{shapeType=Box, extentX=2.037, extentY=2.037, extentZ=1.7394}]
 dyn_StringLights_l_a01 [{shapeType=Box, extentX=2.8937, extentY=5.2026, extentZ=4.0613, center=<0 -0.5685 -1.7866>, rotation=<rot 0 0 24.7475>}]
 dyn_StringLights_l_b00 [{shapeType=Capsule, length=1, radius=1, rotation=<rot 90 0 0>}]
 dyn_StringLights_l_b01 [{shapeType=Capsule, length=1.5, radius=2, center=<1 0 0>, rotation=<rot 90 0 0>}]
 dyn_StringLights_l_b02 [{shapeType=Capsule, length=1.5, radius=2, center=<1 0 0>, rotation=<rot 90 0 0>}]
 dyn_StringLights_l_b03 [{shapeType=Capsule, length=1.5, radius=2, center=<1 0 0>, rotation=<rot 90 0 0>}]
 dyn_StringLights_l_c00 [{shapeType=Box, extentX=2.037, extentY=2.037, extentZ=1.7394}]
 dyn_StringLights_l_c01 [{shapeType=Box, extentX=2.8937, extentY=7.5, extentZ=4.0613, center=<0 0 -1.1381>}]
 dyn_StringLights_r_a00 [{shapeType=Box, extentX=2.037, extentY=2.037, extentZ=1.7394}]
 dyn_StringLights_r_a01 [{shapeType=Box, extentX=2.8937, extentY=5.2026, extentZ=4.0613, center=<0 -0.0001 -1.9529>}]
 dyn_StringLights_r_b00 [{shapeType=Capsule, length=4, radius=2.5, center=<1 0 0>, rotation=<rot 90 0 0>}]
 dyn_StringLights_r_b01 [{shapeType=Capsule, length=4, radius=2.5, center=<1 0 0>, rotation=<rot 90 0 0>}]
 dyn_StringLights_r_b02 [{shapeType=Capsule, length=4, radius=2.5, center=<1 0 0>, rotation=<rot 90 0 0>}]
 dyn_StringLights_r_c00 [{shapeType=Capsule, length=1.25, radius=2, center=<1.5 0 0>, rotation=<rot 90 0 0>}]
 dyn_StringLights_r_c01 [{shapeType=Capsule, length=1.25, radius=2, center=<1.5 0 0>, rotation=<rot 90 0 0>}]
 dyn_StringLights_r_c02 [{shapeType=Capsule, length=1.25, radius=2, center=<1.5 0 0>, rotation=<rot 90 0 0>}]
 dyn_head [{shapeType=Capsule, length=0.505, radius=0.505, rotation=<rot 0 -90 0>}]
}
constraints [
 {bone1Name=dyn_StringLights_l_a00, bone2Name=dyn_head, swing1LimitDegrees=2, swing2LimitDegrees=2, twistLimitDegrees=0.5}
 {bone1Name=dyn_StringLights_l_a01, bone2Name=dyn_StringLights_l_a00, swing1LimitDegrees=2, swing2LimitDegrees=2, twistLimitDegrees=0.5}
 {bone1Name=dyn_StringLights_l_b00, bone2Name=dyn_head, swing1LimitDegrees=45, swing2LimitDegrees=45, twistLimitDegrees=45}
 {bone1Name=dyn_StringLights_l_c00, bone2Name=dyn_head, swing1LimitDegrees=10, swing2LimitDegrees=10, twistLimitDegrees=10}
 {bone1Name=dyn_StringLights_l_c01, bone2Name=dyn_StringLights_l_c00, swing1LimitDegrees=10, swing2LimitDegrees=10, twistLimitDegrees=10}
 {bone1Name=dyn_StringLights_r_a00, bone2Name=dyn_head, swing1LimitDegrees=5, swing2LimitDegrees=5, twistLimitDegrees=0}
 {bone1Name=dyn_StringLights_r_a01, bone2Name=dyn_StringLights_r_a00, swing1LimitDegrees=5, swing2LimitDegrees=5, twistLimitDegrees=0}
 {bone1Name=dyn_StringLights_r_b00, bone2Name=dyn_head, swing1LimitDegrees=65, swing2LimitDegrees=65, twistLimitDegrees=5}
 {bone1Name=dyn_StringLights_r_b01, bone2Name=dyn_StringLights_r_b00, swing1LimitDegrees=30, swing2LimitDegrees=30, twistLimitDegrees=5}
 {bone1Name=dyn_StringLights_r_b02, bone2Name=dyn_StringLights_r_b01, swing1LimitDegrees=10, swing2LimitDegrees=10, twistLimitDegrees=5}
 {bone1Name=dyn_StringLights_r_c00, bone2Name=dyn_head, swing1LimitDegrees=45, swing2LimitDegrees=45, twistLimitDegrees=10}
 {bone1Name=dyn_StringLights_r_c01, bone2Name=dyn_StringLights_r_c00, swing1LimitDegrees=30, swing2LimitDegrees=30, twistLimitDegrees=10}
 {bone1Name=dyn_StringLights_r_c02, bone2Name=dyn_StringLights_r_c01, swing1LimitDegrees=10, swing2LimitDegrees=10, twistLimitDegrees=10}
 {bone1Name=dyn_StringLights_l_b01, bone2Name=dyn_StringLights_l_b00, swing1LimitDegrees=66, swing2LimitDegrees=45, twistLimitDegrees=45}
 {bone1Name=dyn_StringLights_l_b02, bone2Name=dyn_StringLights_l_b01, swing1LimitDegrees=45, swing2LimitDegrees=45, twistLimitDegrees=45}
 {bone1Name=dyn_StringLights_l_b03, bone2Name=dyn_StringLights_l_b02, swing1LimitDegrees=45, swing2LimitDegrees=45, twistLimitDegrees=45}
 {
  bone1Name=dyn_head
  bone2Name=Melee_TwoHanded_Handle
  swing1LimitDegrees=45
  swing1Motion=Free
  swing2LimitDegrees=45
  swing2Motion=Free
  twistLimitDegrees=45
  twistMotion=Free
 }
]
