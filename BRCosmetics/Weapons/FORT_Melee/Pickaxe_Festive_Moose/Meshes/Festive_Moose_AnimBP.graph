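# Animation Blueprint for the Festive Moose pickaxe that adds secondary physics motion to the weapon's skeleton.
# A RigidBody node simulates its physics asset in the root bone's space, with overridden gravity and damped component motion.
# The left light bone is then constrained to a weighted blend of handle, string and string-light bones, and additively scaled.
# Each update computes drive_alpha from a string-light bone's twist relative to the head, but the AnimGraph never reads it.
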AnimBlueprint /BRCosmetics/Weapons/FORT_Melee/Pickaxe_Festive_Moose/Meshes/Festive_Moose_AnimBP
target_skeleton=&/Festive_Moose_Skeleton
graphs [
 AnimGraph {nodes [
   AnimGraphNode_SubInput_0 = Input Pose()
   n1 = RigidBody(ComponentPose=Local To Component(LocalPose=@AnimGraphNode_SubInput_0)) {
    bClampLinearTranslationLimitToRefPose=true
    bOverrideWorldGravity=true
    overridePhysicsAsset=&/Festive_Moose_Physics
    simulationSpace=BaseBoneSpace
    alphaBoolBlend {blendOption=Linear}
    baseBoneRef {boneName=b_AR_Root}
    componentAppliedLinearAccClamp=<1200 1200 1000>
    componentLinearAccScale=<0.45 0.3 0.2>
    componentLinearVelScale=<0.3 0.3 0.2>
    overrideWorldGravity {z=-980}
    simSpaceSettings {velocityScaleZ=0.1, worldAlpha=0.35, externalLinearDragV {z=0.7}}
   }
   n2 = "Constraint - dyn_LightA_01_l (PPPPP)"!(
    ComponentPose=@n1
    Alpha=1
    ConstraintWeights_0=0.158
    ConstraintWeights_1:float=0.2823
    ConstraintWeights_2:float=0.0932
    ConstraintWeights_3:float=0.1789
    ConstraintWeights_4:float=0.2877
   ) -> (Pose) {
    alphaBoolBlend {blendOption=Linear}
    boneToModify {boneName=dyn_LightA_01_l}
    constraintSetup [
     {targetBone {boneName=Melee_TwoHanded_Handle}}
     {targetBone {boneName=dyn_StringA_01_l}}
     {targetBone {boneName=dyn_StringLights_l_b01}}
     {targetBone {boneName=dyn_StringLights_l_b02}}
     {targetBone {boneName=dyn_StringLights_l_b03}}
    ]
    constraintWeights [1,1,1,1,1]
   }
   n3 = "Transform (Modify) Bone - Bone: dyn_LightA_01_l"(ComponentPose=@n2, Alpha=1, Translation="0,0,0", Rotation="0,0,0", Scale="0.9,0.9,0.9") {
    scaleMode=BMM_Additive
    scaleSpace=BCS_BoneSpace
    alphaBoolBlend {blendOption=Linear}
    boneToModify {boneName=dyn_LightA_01_l}
    scale=<0.9 0.9 0.9>
   }
   Output Pose(Result=Component To Local(ComponentPose=@n3)) {layerGroup=None}
  ]}
 EventGraph {nodes [
   Event BlueprintUpdateAnimation()
   > Get!(execute:Exec) -> (then:Exec, else:Exec, mesh:Skeletal Mesh Component Object Reference)
   > Set drive_alpha(drive_alpha:double=GetAlphaFromReferenceTransformDelta(
     SkeletalMeshComponent=Get mesh() -> (mesh:Skeletal Mesh Component Object Reference)
     BoneName=dyn_StringLights_l_b01
     BaseName=dyn_head
     DeltaComponentType=QuaternionTwist
     InputRange=(X=0,Y=100)
     TwistAxis="0,0,1"
    )) -> (Output_Get:double)
   Event BlueprintInitializeAnimation()
   > Set mesh(mesh:Skeletal Mesh Component Object Reference=GetOwnerSkeletalMeshComponent()) -> (Output_Get:Skeletal Mesh Component Object Reference)
  ]}
]
variables [
 drive_alpha {type=number}
 mesh {type=SkeletalMeshComponent}
]
